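# Sound cue for Husky Husk attack-building effort vocals, played as non-critical enemy audio with Husk attenuation and a prevent-new concurrency override.
# Group controls of size one choose between a mix of attack vocal plus low-pitched, quieter sweetener, and a quieter, pitch-varied crowd vocal for overlapping instances.
# Random nodes pick each wave.
# The second quality level repeats this structure with fewer waves; the last two levels are empty.
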SoundCue /Game/Sounds/Fort_Monster_Husky_Husk/Efforts/Husky_Husk_Attack_Building_Cue
root=QualityLevel_0
nodes [
 {id=GroupControl_1, children [Mixer_1,Modulator_5], properties {groupSizes [1]}}
 {id=GroupControl_2, children [Mixer_2,Modulator_9], properties {groupSizes [1]}}
 {id=Mixer_1, children [Modulator_4,Modulator_7], properties {inputVolume [1,0.7]}}
 {id=Mixer_2, children [Modulator_8,Modulator_11], properties {inputVolume [1,0.7]}}
 {id=Modulator_11, children [Random_8], properties {pitchMax=0.8, pitchMin=0.7}}
 {id=Modulator_4, children [Random_7]}
 {id=Modulator_5, children [Random_6], properties {pitchMax=1.3, volumeMax=0.8, volumeMin=0.4}}
 {id=Modulator_7, children [Random_4], properties {pitchMax=0.8, pitchMin=0.7}}
 {id=Modulator_8, children [Random_11]}
 {id=Modulator_9, children [Random_10], properties {pitchMax=1.3, volumeMax=0.8, volumeMin=0.4}}
 {id=QualityLevel_0, children [GroupControl_1, GroupControl_2, null, null]}
 {
  id=Random_10
  children [WavePlayer_72, WavePlayer_74, WavePlayer_75, WavePlayer_81, WavePlayer_84, WavePlayer_85, WavePlayer_91, WavePlayer_92, WavePlayer_93]
  properties {weights [1,1,1,1,1,1,1,1,1]}
 }
 {id=Random_11, children [WavePlayer_94, WavePlayer_95, WavePlayer_96, WavePlayer_98, WavePlayer_100], properties {weights [1,1,1,1,1]}}
 {id=Random_4, children [WavePlayer_34, WavePlayer_35, WavePlayer_36, WavePlayer_37], properties {weights [1,1,1,1]}}
 {
  id=Random_6
  children [
   WavePlayer_38
   WavePlayer_39
   WavePlayer_40
   WavePlayer_41
   WavePlayer_42
   WavePlayer_43
   WavePlayer_44
   WavePlayer_45
   WavePlayer_46
   WavePlayer_47
   WavePlayer_48
   WavePlayer_49
   WavePlayer_50
   WavePlayer_51
   WavePlayer_52
   WavePlayer_53
   WavePlayer_54
   WavePlayer_55
   WavePlayer_56
   WavePlayer_57
   WavePlayer_58
   WavePlayer_59
  ]
  properties {weights [1,1,1,1,1,1,1,1,1,1,1,1,1,1,1,1,1,1,1,1,1,1]}
 }
 {
  id=Random_7
  children [WavePlayer_60, WavePlayer_61, WavePlayer_62, WavePlayer_63, WavePlayer_64, WavePlayer_65, WavePlayer_66, WavePlayer_67]
  properties {weights [1,1,1,1,1,1,1,1]}
 }
 {id=Random_8, children [WavePlayer_68, WavePlayer_69, WavePlayer_71], properties {weights [1,1,1]}}
 {id=WavePlayer_100, properties {soundWaveAssetPtr=&/husky_husk_attack_07}}
 {id=WavePlayer_34, properties {soundWaveAssetPtr=/Game/Sounds/Fort_Monster_Husk/Husk_Efforts/husk_pain_sweeten_01}}
 {id=WavePlayer_35, properties {soundWaveAssetPtr=/Game/Sounds/Fort_Monster_Husk/Husk_Efforts/husk_attack_sweeten_01}}
 {id=WavePlayer_36, properties {soundWaveAssetPtr=/Game/Sounds/Fort_Monster_Husk/Husk_Efforts/husk_attack_sweeten_02}}
 {id=WavePlayer_37, properties {soundWaveAssetPtr=/Game/Sounds/Fort_Monster_Husk/Husk_Efforts/husk_attack_sweeten_03}}
 {id=WavePlayer_38, properties {soundWaveAssetPtr=/Game/Sounds/Fort_Monster_Husk/Husk_Efforts/husk_newcrowd_10}}
 {id=WavePlayer_39, properties {soundWaveAssetPtr=/Game/Sounds/Fort_Monster_Husk/Husk_Efforts/husk_newcrowd_11}}
 {id=WavePlayer_40, properties {soundWaveAssetPtr=/Game/Sounds/Fort_Monster_Husk/Husk_Efforts/husk_newcrowd_12}}
 {id=WavePlayer_41, properties {soundWaveAssetPtr=/Game/Sounds/Fort_Monster_Husk/Husk_Efforts/husk_newcrowd_13}}
 {id=WavePlayer_42, properties {soundWaveAssetPtr=/Game/Sounds/Fort_Monster_Husk/Husk_Efforts/husk_newcrowd_14}}
 {id=WavePlayer_43, properties {soundWaveAssetPtr=/Game/Sounds/Fort_Monster_Husk/Husk_Efforts/husk_newcrowd_15}}
 {id=WavePlayer_44, properties {soundWaveAssetPtr=/Game/Sounds/Fort_Monster_Husk/Husk_Efforts/husk_newcrowd_16}}
 {id=WavePlayer_45, properties {soundWaveAssetPtr=/Game/Sounds/Fort_Monster_Husk/Husk_Efforts/husk_newcrowd_17}}
 {id=WavePlayer_46, properties {soundWaveAssetPtr=/Game/Sounds/Fort_Monster_Husk/Husk_Efforts/husk_newcrowd_18}}
 {id=WavePlayer_47, properties {soundWaveAssetPtr=/Game/Sounds/Fort_Monster_Husk/Husk_Efforts/husk_newcrowd_19}}
 {id=WavePlayer_48, properties {soundWaveAssetPtr=/Game/Sounds/Fort_Monster_Husk/Husk_Efforts/husk_newcrowd_20}}
 {id=WavePlayer_49, properties {soundWaveAssetPtr=/Game/Sounds/Fort_Monster_Husk/Husk_Efforts/husk_newcrowd_21}}
 {id=WavePlayer_50, properties {soundWaveAssetPtr=/Game/Sounds/Fort_Monster_Husk/Husk_Efforts/husk_newcrowd_22}}
 {id=WavePlayer_51, properties {soundWaveAssetPtr=/Game/Sounds/Fort_Monster_Husk/Husk_Efforts/husk_newcrowd_01}}
 {id=WavePlayer_52, properties {soundWaveAssetPtr=/Game/Sounds/Fort_Monster_Husk/Husk_Efforts/husk_newcrowd_02}}
 {id=WavePlayer_53, properties {soundWaveAssetPtr=/Game/Sounds/Fort_Monster_Husk/Husk_Efforts/husk_newcrowd_03}}
 {id=WavePlayer_54, properties {soundWaveAssetPtr=/Game/Sounds/Fort_Monster_Husk/Husk_Efforts/husk_newcrowd_04}}
 {id=WavePlayer_55, properties {soundWaveAssetPtr=/Game/Sounds/Fort_Monster_Husk/Husk_Efforts/husk_newcrowd_05}}
 {id=WavePlayer_56, properties {soundWaveAssetPtr=/Game/Sounds/Fort_Monster_Husk/Husk_Efforts/husk_newcrowd_06}}
 {id=WavePlayer_57, properties {soundWaveAssetPtr=/Game/Sounds/Fort_Monster_Husk/Husk_Efforts/husk_newcrowd_07}}
 {id=WavePlayer_58, properties {soundWaveAssetPtr=/Game/Sounds/Fort_Monster_Husk/Husk_Efforts/husk_newcrowd_08}}
 {id=WavePlayer_59, properties {soundWaveAssetPtr=/Game/Sounds/Fort_Monster_Husk/Husk_Efforts/husk_newcrowd_09}}
 {id=WavePlayer_60, properties {soundWaveAssetPtr=&/husky_husk_attack_01}}
 {id=WavePlayer_61, properties {soundWaveAssetPtr=&/husky_husk_attack_02}}
 {id=WavePlayer_62, properties {soundWaveAssetPtr=&/husky_husk_attack_03}}
 {id=WavePlayer_63, properties {soundWaveAssetPtr=&/husky_husk_attack_04}}
 {id=WavePlayer_64, properties {soundWaveAssetPtr=&/husky_husk_attack_05}}
 {id=WavePlayer_65, properties {soundWaveAssetPtr=&/husky_husk_attack_06}}
 {id=WavePlayer_66, properties {soundWaveAssetPtr=&husky_husk_attack_07}}
 {id=WavePlayer_67, properties {soundWaveAssetPtr=&/husky_husk_attack_08}}
 {id=WavePlayer_68, properties {soundWaveAssetPtr=&husk_pain_sweeten_01}}
 {id=WavePlayer_69, properties {soundWaveAssetPtr=&husk_attack_sweeten_01}}
 {id=WavePlayer_71, properties {soundWaveAssetPtr=&husk_attack_sweeten_03}}
 {id=WavePlayer_72, properties {soundWaveAssetPtr=&husk_newcrowd_10}}
 {id=WavePlayer_74, properties {soundWaveAssetPtr=&husk_newcrowd_12}}
 {id=WavePlayer_75, properties {soundWaveAssetPtr=&husk_newcrowd_13}}
 {id=WavePlayer_81, properties {soundWaveAssetPtr=&husk_newcrowd_19}}
 {id=WavePlayer_84, properties {soundWaveAssetPtr=&husk_newcrowd_22}}
 {id=WavePlayer_85, properties {soundWaveAssetPtr=&husk_newcrowd_01}}
 {id=WavePlayer_91, properties {soundWaveAssetPtr=&husk_newcrowd_07}}
 {id=WavePlayer_92, properties {soundWaveAssetPtr=&husk_newcrowd_08}}
 {id=WavePlayer_93, properties {soundWaveAssetPtr=&husk_newcrowd_09}}
 {id=WavePlayer_94, properties {soundWaveAssetPtr=&husky_husk_attack_01}}
 {id=WavePlayer_95, properties {soundWaveAssetPtr=&husky_husk_attack_02}}
 {id=WavePlayer_96, properties {soundWaveAssetPtr=&husky_husk_attack_03}}
 {id=WavePlayer_98, properties {soundWaveAssetPtr=&husky_husk_attack_05}}
]
properties {
 attenuationSettings=/Game/Sounds/Attenuation/Fort_Monster_Husk_Att
 bOverrideConcurrency=true
 maxDistance=2800
 soundClassObject=/Game/Sounds/FortSoundClassesAndModes/Fort_Enemy_NonCritical
 volumeMultiplier=0.6
 attenuationOverrides {distanceAlgorithm=NaturalSound, falloffDistance=4200, attenuationShapeExtents {x=800}}
 concurrencyOverrides {resolutionRule=PreventNew}
}
