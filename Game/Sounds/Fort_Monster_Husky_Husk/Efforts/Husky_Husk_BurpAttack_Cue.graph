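# Sound cue for the Husky Husk enemy's burp attack, routed to the enemy attack sound class with Husk directional-cone attenuation.
# Each variant mixes a random, pitch-varied husky burp with a quieter, higher-pitched "sweet" burp that follows after a short random delay.
# A quality-level root selects between a fuller first variant and a second with fewer wave choices, leaving the last two slots empty.
# Its concurrency override prevents new instances at the limit.
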SoundCue /Game/Sounds/Fort_Monster_Husky_Husk/Efforts/Husky_Husk_BurpAttack_Cue
root=QualityLevel_1
nodes [
 {id=Delay_1, children [Modulator_3], properties {delayMax=0.7, delayMin=0.5}}
 {id=Delay_2, children [Modulator_5], properties {delayMax=0.7, delayMin=0.5}}
 {id=Mixer_1, children [Modulator_2,Delay_1], properties {inputVolume [1,0.3]}}
 {id=Mixer_2, children [Modulator_4,Delay_2], properties {inputVolume [1,0.3]}}
 {id=Modulator_2, children [Random_2], properties {pitchMin=0.7}}
 {id=Modulator_3, children [Random_3], properties {pitchMax=1.3, pitchMin=1}}
 {id=Modulator_4, children [Random_4], properties {pitchMin=0.7}}
 {id=Modulator_5, children [Random_5], properties {pitchMax=1.3, pitchMin=1}}
 {id=QualityLevel_1, children [Mixer_1,Mixer_2,null,null]}
 {id=Random_2, children [WavePlayer_8, WavePlayer_9, WavePlayer_10, WavePlayer_11, WavePlayer_12], properties {weights [1,1,1,1,1]}}
 {id=Random_3, children [WavePlayer_13, WavePlayer_14, WavePlayer_15], properties {weights [1,1,1]}}
 {id=Random_4, children [WavePlayer_16, WavePlayer_17, WavePlayer_19], properties {weights [1,1,1]}}
 {id=Random_5, children [WavePlayer_21, WavePlayer_22], properties {weights [1,1]}}
 {id=WavePlayer_10, properties {soundWaveAssetPtr=/Game/Sounds/Fort_Monster_Husk/SFX/husk_husky_burp_03}}
 {id=WavePlayer_11, properties {soundWaveAssetPtr=/Game/Sounds/Fort_Monster_Husk/SFX/husk_husky_burp_04}}
 {id=WavePlayer_12, properties {soundWaveAssetPtr=/Game/Sounds/Fort_Monster_Husk/SFX/husk_husky_burp_05}}
 {id=WavePlayer_13, properties {soundWaveAssetPtr=/Game/Sounds/Fort_Monster_Husk/SFX/burp_sweet_01}}
 {id=WavePlayer_14, properties {soundWaveAssetPtr=/Game/Sounds/Fort_Monster_Husk/SFX/burp_sweet_03}}
 {id=WavePlayer_15, properties {soundWaveAssetPtr=/Game/Sounds/Fort_Monster_Husk/SFX/burp_sweet_02}}
 {id=WavePlayer_16, properties {soundWaveAssetPtr=/Game/Sounds/Fort_Monster_Husk/SFX/husk_husky_burp_01}}
 {id=WavePlayer_17, properties {soundWaveAssetPtr=/Game/Sounds/Fort_Monster_Husk/SFX/husk_husky_burp_02}}
 {id=WavePlayer_19, properties {soundWaveAssetPtr=&husk_husky_burp_04}}
 {id=WavePlayer_21, properties {soundWaveAssetPtr=&burp_sweet_01}}
 {id=WavePlayer_22, properties {soundWaveAssetPtr=&burp_sweet_03}}
 {id=WavePlayer_8, properties {soundWaveAssetPtr=&husk_husky_burp_01}}
 {id=WavePlayer_9, properties {soundWaveAssetPtr=&husk_husky_burp_02}}
]
properties {
 attenuationSettings=/Game/Sounds/Attenuation/Fort_Monster_HuskDirectionalCone_Att
 bOverrideConcurrency=true
 maxDistance=966.4637
 soundClassObject=/Game/Sounds/FortSoundClassesAndModes/Fort_Enemy_Attack
 volumeMultiplier=2.1
 attenuationOverrides {distanceAlgorithm=Logarithmic}
 concurrencyOverrides {resolutionRule=PreventNew}
}
