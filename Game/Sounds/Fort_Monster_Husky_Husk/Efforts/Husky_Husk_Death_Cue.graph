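# Sound cue for the Husky Husk enemy's death, layering a random death or pain vocal over a quieter electrical zap.
# A quality switch feeds two such mixes, the second drawing from a smaller vocal pool, and leaves two slots empty.
# Each layer gets random pitch modulation and its own natural-sound attenuation override.
# The cue uses the enemy-damaged sound class and a shared concurrency rule set.
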SoundCue /Game/Sounds/Fort_Monster_Husky_Husk/Efforts/Husky_Husk_Death_Cue
root=QualityLevel_0
nodes [
 {
  id=Attenuation_2
  children [Modulator_2]
  properties {bOverrideAttenuation=true, attenuationOverrides {distanceAlgorithm=NaturalSound, falloffDistance=5000, attenuationShapeExtents {x=1000}}}
 }
 {
  id=Attenuation_3
  children [Modulator_3]
  properties {bOverrideAttenuation=true, attenuationOverrides {distanceAlgorithm=NaturalSound, falloffDistance=5000, attenuationShapeExtents {x=1000}}}
 }
 {
  id=Attenuation_4
  children [Modulator_4]
  properties {bOverrideAttenuation=true, attenuationOverrides {distanceAlgorithm=NaturalSound, falloffDistance=5000, attenuationShapeExtents {x=1000}}}
 }
 {
  id=Attenuation_5
  children [Modulator_5]
  properties {bOverrideAttenuation=true, attenuationOverrides {distanceAlgorithm=NaturalSound, falloffDistance=5000, attenuationShapeExtents {x=1000}}}
 }
 {id=Mixer_1, children [Attenuation_3, Attenuation_2], properties {inputVolume [0.6,2]}}
 {id=Mixer_2, children [Attenuation_5, Attenuation_4], properties {inputVolume [0.6,2]}}
 {id=Modulator_2, children [Random_1], properties {pitchMax=1.5, pitchMin=0.9}}
 {id=Modulator_3, children [WavePlayer_13], properties {pitchMax=1.6, pitchMin=0.9}}
 {id=Modulator_4, children [Random_2], properties {pitchMax=1.5, pitchMin=0.9}}
 {id=Modulator_5, children [WavePlayer_20], properties {pitchMax=1.6, pitchMin=0.9}}
 {id=QualityLevel_0, children [Mixer_1,Mixer_2,null,null]}
 {
  id=Random_1
  children [WavePlayer_7, WavePlayer_8, WavePlayer_9, WavePlayer_10, WavePlayer_11, WavePlayer_12]
  properties {weights [1,1,1,1,1,1]}
 }
 {id=Random_2, children [WavePlayer_14, WavePlayer_15, WavePlayer_18], properties {weights [1,1,1]}}
 {id=WavePlayer_10, properties {soundWaveAssetPtr=&/husky_husk_death_B_03}}
 {id=WavePlayer_11, properties {soundWaveAssetPtr=&/husky_husk_death_B_04}}
 {id=WavePlayer_12, properties {soundWaveAssetPtr=&/husky_husk_death_B_05}}
 {id=WavePlayer_13, properties {soundWaveAssetPtr=/Game/Sounds/Fort_Monster_Husk/SFX/husk_death_zap_B_01}}
 {id=WavePlayer_14, properties {soundWaveAssetPtr=&/husky_husk_pain_B_08}}
 {id=WavePlayer_15, properties {soundWaveAssetPtr=&/husky_husk_death_B_01}}
 {id=WavePlayer_18, properties {soundWaveAssetPtr=&husky_husk_death_B_04}}
 {id=WavePlayer_20, properties {soundWaveAssetPtr=&husk_death_zap_B_01}}
 {id=WavePlayer_7, properties {soundWaveAssetPtr=&husky_husk_pain_B_08}}
 {id=WavePlayer_8, properties {soundWaveAssetPtr=&husky_husk_death_B_01}}
 {id=WavePlayer_9, properties {soundWaveAssetPtr=&/husky_husk_death_B_02}}
]
properties {
 attenuationSettings=/Game/Sounds/Attenuation/Fort_Monster_Husk_PainDeath_Att
 maxDistance=6000
 soundClassObject=/Game/Sounds/FortSoundClassesAndModes/Fort_Enemy_Damaged_PSM
 volumeMultiplier=0.8
 concurrencyOverrides {resolutionRule=PreventNew}
 concurrencySet [/Game/Sounds/Concurrency_Rules/TwoMaximum_StopFurthest_VolumeScale]
}
