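# Sound cue for the effort vocal a Husky Husk enemy makes when launched, as its name indicates.
# A quality switch picks between two mixers: the first layers random Husky pain clips, the second random Husky wake-up clips from a smaller pool.
# Both blend in a quieter, pitched-down random regular Husk wake effort, with modulators varying each layer.
# It uses Husk attenuation, the non-critical enemy sound class and prevent-new concurrency.
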SoundCue /Game/Sounds/Fort_Monster_Husky_Husk/Efforts/Husky_Husk_Launched_Cue
root=QualityLevel_0
nodes [
 {id=Mixer_1, children [Modulator_2,Modulator_3], properties {inputVolume [1,0.8]}}
 {id=Mixer_2, children [Modulator_4,Modulator_5], properties {inputVolume [1,0.8]}}
 {id=Modulator_2, children [Random_3]}
 {id=Modulator_3, children [Random_2], properties {pitchMax=0.8, pitchMin=0.7}}
 {id=Modulator_4, children [Random_5]}
 {id=Modulator_5, children [Random_4], properties {pitchMax=0.8, pitchMin=0.7}}
 {id=QualityLevel_0, children [Mixer_1,Mixer_2,null,null]}
 {id=Random_2, children [WavePlayer_8, WavePlayer_9, WavePlayer_10, WavePlayer_11], properties {weights [1,1,1,1]}}
 {id=Random_3, children [WavePlayer_36, WavePlayer_37, WavePlayer_38, WavePlayer_39], properties {weights [1,1,1,1]}}
 {id=Random_4, children [WavePlayer_16, WavePlayer_18, WavePlayer_19], properties {weights [1,1,1]}}
 {id=Random_5, children [WavePlayer_20, WavePlayer_22], properties {weights [1,1]}}
 {id=WavePlayer_10, properties {soundWaveAssetPtr=/Game/Sounds/Fort_Monster_Husk/Husk_Efforts/Husk_WakeSweet_03}}
 {id=WavePlayer_11, properties {soundWaveAssetPtr=/Game/Sounds/Fort_Monster_Husk/Husk_Efforts/Husk_WakeSweet_05}}
 {id=WavePlayer_16, properties {soundWaveAssetPtr=/Game/Sounds/Fort_Monster_Husk/Husk_Efforts/Husk_WakeSweet_01}}
 {id=WavePlayer_18, properties {soundWaveAssetPtr=&Husk_WakeSweet_03}}
 {id=WavePlayer_19, properties {soundWaveAssetPtr=&Husk_WakeSweet_05}}
 {id=WavePlayer_20, properties {soundWaveAssetPtr=&/husky_husk_wakeup_01}}
 {id=WavePlayer_22, properties {soundWaveAssetPtr=&/husky_husk_wakeup_03}}
 {id=WavePlayer_36, properties {soundWaveAssetPtr=&/husky_husk_pain_B_01}}
 {id=WavePlayer_37, properties {soundWaveAssetPtr=&/husky_husk_pain_B_02}}
 {id=WavePlayer_38, properties {soundWaveAssetPtr=&/husky_husk_pain_B_03}}
 {id=WavePlayer_39, properties {soundWaveAssetPtr=&/husky_husk_pain_B_04}}
 {id=WavePlayer_8, properties {soundWaveAssetPtr=&Husk_WakeSweet_01}}
 {id=WavePlayer_9, properties {soundWaveAssetPtr=/Game/Sounds/Fort_Monster_Husk/Husk_Efforts/Husk_WakeSweet_02}}
]
properties {
 attenuationSettings=/Game/Sounds/Attenuation/Fort_Monster_Husk_Att
 bOverrideConcurrency=true
 maxDistance=2800
 soundClassObject=/Game/Sounds/FortSoundClassesAndModes/Fort_Enemy_NonCritical
 volumeMultiplier=1
 attenuationOverrides {distanceAlgorithm=NaturalSound, falloffDistance=4000, attenuationShapeExtents {x=1000}}
 concurrencyOverrides {resolutionRule=PreventNew}
}
