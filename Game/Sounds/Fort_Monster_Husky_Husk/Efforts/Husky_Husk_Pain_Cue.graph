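# Sound cue for Husky Husk pain vocalizations, using the enemy-damaged sound class, Husk attenuation and a two-voice concurrency rule.
# A quality-level node fills two of four slots, the second using smaller wave pools and a louder sweetener mix.
# Each branch mixes random Husky pain takes with quieter, pitched-down regular Husk sweeteners.
# Nested modulators randomize pitch per layer and again over each mix.
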SoundCue /Game/Sounds/Fort_Monster_Husky_Husk/Efforts/Husky_Husk_Pain_Cue
root=QualityLevel_0
nodes [
 {id=Mixer_1, children [Modulator_3,Modulator_4], properties {inputVolume [1,0.25]}}
 {id=Mixer_2, children [Modulator_6,Modulator_7], properties {inputVolume [1,0.7]}}
 {id=Modulator_3, children [Random_3], properties {pitchMax=1.1}}
 {id=Modulator_4, children [Random_2], properties {pitchMax=0.8, pitchMin=0.7}}
 {id=Modulator_5, children [Mixer_1], properties {pitchMax=1.5, pitchMin=1}}
 {id=Modulator_6, children [Random_5], properties {pitchMax=1.1}}
 {id=Modulator_7, children [Random_4], properties {pitchMax=0.8, pitchMin=0.7}}
 {id=Modulator_8, children [Mixer_2], properties {pitchMax=1.5, pitchMin=1}}
 {id=QualityLevel_0, children [Modulator_5,Modulator_8,null,null]}
 {
  id=Random_2
  children [
   WavePlayer_22
   WavePlayer_23
   WavePlayer_24
   WavePlayer_25
   WavePlayer_26
   WavePlayer_27
   WavePlayer_28
   WavePlayer_29
   WavePlayer_30
   WavePlayer_31
   WavePlayer_32
   WavePlayer_33
   WavePlayer_34
   WavePlayer_35
  ]
  properties {weights [1,1,1,1,1,1,1,1,1,1,1,1,1,1]}
 }
 {
  id=Random_3
  children [WavePlayer_36, WavePlayer_37, WavePlayer_38, WavePlayer_39, WavePlayer_40, WavePlayer_41, WavePlayer_42, WavePlayer_43]
  properties {weights [1,1,1,1,1,1,1,1]}
 }
 {
  id=Random_4
  children [WavePlayer_45, WavePlayer_46, WavePlayer_47, WavePlayer_49, WavePlayer_51, WavePlayer_53, WavePlayer_54, WavePlayer_56]
  properties {weights [1,1,1,1,1,1,1,1]}
 }
 {
  id=Random_5
  children [WavePlayer_58, WavePlayer_60, WavePlayer_61, WavePlayer_63, WavePlayer_64, WavePlayer_65]
  properties {weights [1,1,1,1,1,1]}
 }
 {id=WavePlayer_22, properties {soundWaveAssetPtr=/Game/Sounds/Fort_Monster_Husk/Husk_Efforts/husk_pain_sweeten_10}}
 {id=WavePlayer_23, properties {soundWaveAssetPtr=/Game/Sounds/Fort_Monster_Husk/Husk_Efforts/husk_pain_sweeten_11}}
 {id=WavePlayer_24, properties {soundWaveAssetPtr=/Game/Sounds/Fort_Monster_Husk/Husk_Efforts/husk_pain_sweeten_12}}
 {id=WavePlayer_25, properties {soundWaveAssetPtr=/Game/Sounds/Fort_Monster_Husk/Husk_Efforts/husk_pain_sweeten_13}}
 {id=WavePlayer_26, properties {soundWaveAssetPtr=/Game/Sounds/Fort_Monster_Husk/Husk_Efforts/husk_pain_sweeten_14}}
 {id=WavePlayer_27, properties {soundWaveAssetPtr=/Game/Sounds/Fort_Monster_Husk/Husk_Efforts/husk_pain_sweeten_15}}
 {id=WavePlayer_28, properties {soundWaveAssetPtr=/Game/Sounds/Fort_Monster_Husk/Husk_Efforts/husk_pain_sweeten_16}}
 {id=WavePlayer_29, properties {soundWaveAssetPtr=/Game/Sounds/Fort_Monster_Husk/Husk_Efforts/husk_pain_sweeten_01}}
 {id=WavePlayer_30, properties {soundWaveAssetPtr=/Game/Sounds/Fort_Monster_Husk/Husk_Efforts/husk_pain_sweeten_04}}
 {id=WavePlayer_31, properties {soundWaveAssetPtr=/Game/Sounds/Fort_Monster_Husk/Husk_Efforts/husk_pain_sweeten_05}}
 {id=WavePlayer_32, properties {soundWaveAssetPtr=/Game/Sounds/Fort_Monster_Husk/Husk_Efforts/husk_pain_sweeten_06}}
 {id=WavePlayer_33, properties {soundWaveAssetPtr=/Game/Sounds/Fort_Monster_Husk/Husk_Efforts/husk_pain_sweeten_07}}
 {id=WavePlayer_34, properties {soundWaveAssetPtr=/Game/Sounds/Fort_Monster_Husk/Husk_Efforts/husk_pain_sweeten_08}}
 {id=WavePlayer_35, properties {soundWaveAssetPtr=/Game/Sounds/Fort_Monster_Husk/Husk_Efforts/husk_pain_sweeten_09}}
 {id=WavePlayer_36, properties {soundWaveAssetPtr=&/husky_husk_pain_B_01}}
 {id=WavePlayer_37, properties {soundWaveAssetPtr=&/husky_husk_pain_B_02}}
 {id=WavePlayer_38, properties {soundWaveAssetPtr=&/husky_husk_pain_B_03}}
 {id=WavePlayer_39, properties {soundWaveAssetPtr=&/husky_husk_pain_B_04}}
 {id=WavePlayer_40, properties {soundWaveAssetPtr=&/husky_husk_pain_B_05}}
 {id=WavePlayer_41, properties {soundWaveAssetPtr=&/husky_husk_pain_B_06}}
 {id=WavePlayer_42, properties {soundWaveAssetPtr=&/husky_husk_pain_B_07}}
 {id=WavePlayer_43, properties {soundWaveAssetPtr=&/husky_husk_pain_B_08}}
 {id=WavePlayer_45, properties {soundWaveAssetPtr=&husk_pain_sweeten_11}}
 {id=WavePlayer_46, properties {soundWaveAssetPtr=&husk_pain_sweeten_12}}
 {id=WavePlayer_47, properties {soundWaveAssetPtr=&husk_pain_sweeten_13}}
 {id=WavePlayer_49, properties {soundWaveAssetPtr=&husk_pain_sweeten_15}}
 {id=WavePlayer_51, properties {soundWaveAssetPtr=&husk_pain_sweeten_01}}
 {id=WavePlayer_53, properties {soundWaveAssetPtr=&husk_pain_sweeten_05}}
 {id=WavePlayer_54, properties {soundWaveAssetPtr=&husk_pain_sweeten_06}}
 {id=WavePlayer_56, properties {soundWaveAssetPtr=&husk_pain_sweeten_08}}
 {id=WavePlayer_58, properties {soundWaveAssetPtr=&husky_husk_pain_B_01}}
 {id=WavePlayer_60, properties {soundWaveAssetPtr=&husky_husk_pain_B_03}}
 {id=WavePlayer_61, properties {soundWaveAssetPtr=&husky_husk_pain_B_04}}
 {id=WavePlayer_63, properties {soundWaveAssetPtr=&husky_husk_pain_B_06}}
 {id=WavePlayer_64, properties {soundWaveAssetPtr=&husky_husk_pain_B_07}}
 {id=WavePlayer_65, properties {soundWaveAssetPtr=&husky_husk_pain_B_08}}
]
properties {
 attenuationSettings=/Game/Sounds/Attenuation/Fort_Monster_Husk_Att
 maxDistance=2800
 soundClassObject=/Game/Sounds/FortSoundClassesAndModes/Fort_Enemy_Damaged_PSM
 volumeMultiplier=1.3
 attenuationOverrides {distanceAlgorithm=NaturalSound, falloffDistance=4000, attenuationShapeExtents {x=1000}}
 concurrencyOverrides {resolutionRule=PreventNew}
 concurrencySet [/Game/Sounds/Concurrency_Rules/TwoMaximum_StopFurthest_VolumeScale]
}
