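# Sound cue for the riot shield Husky Husk's laugh vocalizations, using husk monster attenuation.
# A quality-level node feeds the first tier a modulated random pick from ten laugh waves.
# The second tier picks from six waves through two stacked modulators, one adding an upward pitch range, and the remaining two tiers are empty.
# A concurrency override allows one instance and blocks new ones, taking the place of the assigned concurrency set.
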SoundCue /Game/Sounds/Fort_Monster_Husky_Husk/Efforts/Husky_Husk_Riot_Laugh_Cue
root=QualityLevel_0
nodes [
 {id=Modulator_0, children [Random_0]}
 {id=Modulator_1, children [Random_1]}
 {id=Modulator_8, children [Modulator_1], properties {pitchMax=1.5, pitchMin=1}}
 {id=QualityLevel_0, children [Modulator_0,Modulator_8,null,null]}
 {
  id=Random_0
  children [
   WavePlayer_0
   WavePlayer_1
   WavePlayer_2
   WavePlayer_5
   WavePlayer_6
   WavePlayer_7
   WavePlayer_9
   WavePlayer_10
   WavePlayer_11
   WavePlayer_12
  ]
  properties {weights [1,1,1,1,1,1,1,1,1,1]}
 }
 {
  id=Random_1
  children [WavePlayer_13, WavePlayer_14, WavePlayer_15, WavePlayer_16, WavePlayer_17, WavePlayer_18]
  properties {weights [1,1,1,1,1,1]}
 }
 {id=WavePlayer_0, properties {soundWaveAssetPtr=&/Husky_Riot_Shield_Vox_Laughs_09}}
 {id=WavePlayer_1, properties {soundWaveAssetPtr=&/Husky_Riot_Shield_Vox_Laughs_10}}
 {id=WavePlayer_10, properties {soundWaveAssetPtr=&/Husky_Riot_Shield_Vox_Laughs_19}}
 {id=WavePlayer_11, properties {soundWaveAssetPtr=&/Husky_Riot_Shield_Vox_Laughs_07}}
 {id=WavePlayer_12, properties {soundWaveAssetPtr=&/Husky_Riot_Shield_Vox_Laughs_08}}
 {id=WavePlayer_13, properties {soundWaveAssetPtr=&Husky_Riot_Shield_Vox_Laughs_09}}
 {id=WavePlayer_14, properties {soundWaveAssetPtr=&Husky_Riot_Shield_Vox_Laughs_10}}
 {id=WavePlayer_15, properties {soundWaveAssetPtr=&/Husky_Riot_Shield_Vox_Laughs_11}}
 {id=WavePlayer_16, properties {soundWaveAssetPtr=&/Husky_Riot_Shield_Vox_Laughs_12}}
 {id=WavePlayer_17, properties {soundWaveAssetPtr=&/Husky_Riot_Shield_Vox_Laughs_13}}
 {id=WavePlayer_18, properties {soundWaveAssetPtr=&/Husky_Riot_Shield_Vox_Laughs_14}}
 {id=WavePlayer_2, properties {soundWaveAssetPtr=&Husky_Riot_Shield_Vox_Laughs_11}}
 {id=WavePlayer_5, properties {soundWaveAssetPtr=&Husky_Riot_Shield_Vox_Laughs_14}}
 {id=WavePlayer_6, properties {soundWaveAssetPtr=&/Husky_Riot_Shield_Vox_Laughs_15}}
 {id=WavePlayer_7, properties {soundWaveAssetPtr=&/Husky_Riot_Shield_Vox_Laughs_16}}
 {id=WavePlayer_9, properties {soundWaveAssetPtr=&/Husky_Riot_Shield_Vox_Laughs_18}}
]
properties {
 attenuationSettings=/Game/Sounds/Attenuation/Fort_Monster_Husk_Att
 bOverrideConcurrency=true
 maxDistance=2800
 soundClassObject=/Game/Sounds/FortSoundClassesAndModes/Fort_Enemy_Damaged_PSM
 volumeMultiplier=1.3
 attenuationOverrides {distanceAlgorithm=NaturalSound, falloffDistance=4000, attenuationShapeExtents {x=1000}}
 concurrencyOverrides {maxCount=1, resolutionRule=PreventNew}
 concurrencySet [/Game/Sounds/Concurrency_Rules/TwoMaximum_StopFurthest_VolumeScale]
}
