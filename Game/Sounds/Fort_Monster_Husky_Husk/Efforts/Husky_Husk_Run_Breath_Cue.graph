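# Sound cue for the Husky Husk's running breath efforts, built from pursuit vocal waves with husk attenuation and a shared breath concurrency group.
# A quality-level root's first branch plays one of ten modulated waves, with two equally weighted empty slots leaving roughly two in three triggers silent.
# The second branch always picks from six modulated waves, and the remaining two quality levels are empty.
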SoundCue /Game/Sounds/Fort_Monster_Husky_Husk/Efforts/Husky_Husk_Run_Breath_Cue
root=QualityLevel_0
nodes [
 {id=Modulator_1, children [Random_2]}
 {id=Modulator_2, children [Random_4]}
 {id=QualityLevel_0, children [Random_3,Random_5,null,null]}
 {
  id=Random_2
  children [
   WavePlayer_10
   WavePlayer_11
   WavePlayer_12
   WavePlayer_13
   WavePlayer_14
   WavePlayer_15
   WavePlayer_16
   WavePlayer_17
   WavePlayer_18
   WavePlayer_19
  ]
  properties {weights [1,1,1,1,1,1,1,1,1,1]}
 }
 {id=Random_3, children [Modulator_1,null,null], properties {weights [1,1,1]}}
 {
  id=Random_4
  children [WavePlayer_20, WavePlayer_22, WavePlayer_23, WavePlayer_24, WavePlayer_25, WavePlayer_27]
  properties {weights [1,1,1,1,1,1]}
 }
 {id=Random_5, children [Modulator_2], properties {weights [1]}}
 {id=WavePlayer_10, properties {soundWaveAssetPtr=&/husky_husk_pursuit_01_A}}
 {id=WavePlayer_11, properties {soundWaveAssetPtr=&/husky_husk_pursuit_01_B}}
 {id=WavePlayer_12, properties {soundWaveAssetPtr=&/husky_husk_pursuit_03_A}}
 {id=WavePlayer_13, properties {soundWaveAssetPtr=&/husky_husk_pursuit_03_B}}
 {id=WavePlayer_14, properties {soundWaveAssetPtr=&/husky_husk_pursuit_04_A}}
 {id=WavePlayer_15, properties {soundWaveAssetPtr=&/husky_husk_pursuit_04_B}}
 {id=WavePlayer_16, properties {soundWaveAssetPtr=&/husky_husk_pursuit_05_A}}
 {id=WavePlayer_17, properties {soundWaveAssetPtr=&/husky_husk_pursuit_05_B}}
 {id=WavePlayer_18, properties {soundWaveAssetPtr=&/husky_husk_pursuit_06_A}}
 {id=WavePlayer_19, properties {soundWaveAssetPtr=&/husky_husk_pursuit_06_B}}
 {id=WavePlayer_20, properties {soundWaveAssetPtr=&husky_husk_pursuit_01_A}}
 {id=WavePlayer_22, properties {soundWaveAssetPtr=&husky_husk_pursuit_03_A}}
 {id=WavePlayer_23, properties {soundWaveAssetPtr=&husky_husk_pursuit_03_B}}
 {id=WavePlayer_24, properties {soundWaveAssetPtr=&husky_husk_pursuit_04_A}}
 {id=WavePlayer_25, properties {soundWaveAssetPtr=&husky_husk_pursuit_04_B}}
 {id=WavePlayer_27, properties {soundWaveAssetPtr=&husky_husk_pursuit_05_B}}
]
properties {
 attenuationSettings=/Game/Sounds/Attenuation/Fort_Monster_Husk_Att
 maxDistance=2800
 soundClassObject=/Game/Sounds/FortSoundClassesAndModes/Fort_Enemy_NonCritical
 volumeMultiplier=1.3
 attenuationOverrides {distanceAlgorithm=NaturalSound, falloffDistance=4000, attenuationShapeExtents {x=1000}}
 concurrencyOverrides {maxCount=3, resolutionRule=PreventNew}
 concurrencySet [/Game/Sounds/Concurrency_Rules/Husk_Enemy_Group_Breath_Conc]
}
