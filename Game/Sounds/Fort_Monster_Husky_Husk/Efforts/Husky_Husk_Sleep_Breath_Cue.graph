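# Sound cue for the sleep breathing of the Husky Husk enemy, spatialized with Husk attenuation and capped by its own concurrency override.
# Each mix layers a random husky asleep clip at reduced volume with a random regular Husk breath or idle clip, delayed three seconds, pitched down and boosted.
# A quality switch selects between two such mixes, the second with fewer variations, and leaves its remaining two slots empty.
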SoundCue /Game/Sounds/Fort_Monster_Husky_Husk/Efforts/Husky_Husk_Sleep_Breath_Cue
root=QualityLevel_0
nodes [
 {id=Delay_1, children [Modulator_2], properties {delayMax=3, delayMin=3}}
 {id=Delay_2, children [Modulator_4], properties {delayMax=3, delayMin=3}}
 {id=Mixer_1, children [Modulator_3,Delay_1], properties {inputVolume [1,2]}}
 {id=Mixer_2, children [Modulator_5,Delay_2], properties {inputVolume [1,2]}}
 {id=Modulator_2, children [Random_3], properties {pitchMax=0.6, pitchMin=0.6, volumeMax=1.2, volumeMin=1.2}}
 {id=Modulator_3, children [Random_5], properties {pitchMax=1.1, volumeMax=0.7, volumeMin=0.7}}
 {id=Modulator_4, children [Random_6], properties {pitchMax=0.6, pitchMin=0.6, volumeMax=1.2, volumeMin=1.2}}
 {id=Modulator_5, children [Random_8], properties {pitchMax=1.1, volumeMax=0.7, volumeMin=0.7}}
 {id=QualityLevel_0, children [Mixer_1,Mixer_2,null,null]}
 {
  id=Random_3
  children [WavePlayer_13, WavePlayer_12, WavePlayer_11, WavePlayer_14, WavePlayer_15, WavePlayer_16]
  properties {weights [1,1,1,1,1,1]}
 }
 {id=Random_5, children [WavePlayer_17, WavePlayer_18, WavePlayer_19, WavePlayer_20, WavePlayer_21], properties {weights [1,1,1,1,1]}}
 {id=Random_6, children [WavePlayer_24, WavePlayer_23, WavePlayer_25, WavePlayer_26, WavePlayer_27], properties {weights [1,1,1,1,1]}}
 {id=Random_8, children [WavePlayer_28, WavePlayer_29, WavePlayer_30, WavePlayer_31], properties {weights [1,1,1,1]}}
 {id=WavePlayer_11, properties {soundWaveAssetPtr=/Game/Sounds/Fort_Monster_Husk/Husk_Efforts/Husk_Sleep_Breath04}}
 {id=WavePlayer_12, properties {soundWaveAssetPtr=/Game/Sounds/Fort_Monster_Husk/Husk_Efforts/Husk_Sleep_Breath01}}
 {id=WavePlayer_13, properties {soundWaveAssetPtr=/Game/Sounds/Fort_Monster_Husk/Husk_Efforts/Husk_Sleep_Breath05}}
 {id=WavePlayer_14, properties {soundWaveAssetPtr=/Game/Sounds/Fort_Monster_Husk/Husk_Efforts/Husk_Idle01}}
 {id=WavePlayer_15, properties {soundWaveAssetPtr=/Game/Sounds/Fort_Monster_Husk/Husk_Efforts/Husk_Idle04}}
 {id=WavePlayer_16, properties {soundWaveAssetPtr=/Game/Sounds/Fort_Monster_Husk/Husk_Efforts/Husk_Idle03}}
 {id=WavePlayer_17, properties {soundWaveAssetPtr=&/husky_husk_asleep_01}}
 {id=WavePlayer_18, properties {soundWaveAssetPtr=&/husky_husk_asleep_02}}
 {id=WavePlayer_19, properties {soundWaveAssetPtr=&/husky_husk_asleep_03}}
 {id=WavePlayer_20, properties {soundWaveAssetPtr=&/husky_husk_asleep_04}}
 {id=WavePlayer_21, properties {soundWaveAssetPtr=&/husky_husk_asleep_05}}
 {id=WavePlayer_23, properties {soundWaveAssetPtr=&Husk_Sleep_Breath01}}
 {id=WavePlayer_24, properties {soundWaveAssetPtr=&Husk_Sleep_Breath05}}
 {id=WavePlayer_25, properties {soundWaveAssetPtr=&Husk_Idle01}}
 {id=WavePlayer_26, properties {soundWaveAssetPtr=&Husk_Idle04}}
 {id=WavePlayer_27, properties {soundWaveAssetPtr=&Husk_Idle03}}
 {id=WavePlayer_28, properties {soundWaveAssetPtr=&husky_husk_asleep_01}}
 {id=WavePlayer_29, properties {soundWaveAssetPtr=&husky_husk_asleep_02}}
 {id=WavePlayer_30, properties {soundWaveAssetPtr=&husky_husk_asleep_03}}
 {id=WavePlayer_31, properties {soundWaveAssetPtr=&husky_husk_asleep_04}}
]
properties {
 attenuationSettings=/Game/Sounds/Attenuation/Fort_Monster_Husk_Att
 bOverrideConcurrency=true
 maxDistance=2800
 soundClassObject=/Game/Sounds/FortSoundClassesAndModes/Fort_Enemy_NonCritical
 volumeMultiplier=0.9
 attenuationOverrides {distanceAlgorithm=Logarithmic, falloffDistance=3200}
 concurrencyOverrides {maxCount=8, resolutionRule=PreventNew}
}
