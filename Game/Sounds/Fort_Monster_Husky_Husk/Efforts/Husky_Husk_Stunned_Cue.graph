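# Sound cue for the Husky Husk monster's stunned vocal effort.
# A quality-level switch selects between a five-variation branch and a reduced three-variation branch, with two further slots left empty.
# Each branch plays an equally weighted random wave through a Modulator.
# It uses the shared husk attenuation and enemy-damaged sound class, with a concurrency override that prevents new instances at the limit.
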SoundCue /Game/Sounds/Fort_Monster_Husky_Husk/Efforts/Husky_Husk_Stunned_Cue
root=QualityLevel_0
nodes [
 {id=Modulator_1, children [Random_1]}
 {id=Modulator_2, children [Random_2]}
 {id=QualityLevel_0, children [Modulator_1,Modulator_2,null,null]}
 {id=Random_1, children [WavePlayer_5,WavePlayer_6,WavePlayer_7,WavePlayer_8,WavePlayer_9], properties {weights [1,1,1,1,1]}}
 {id=Random_2, children [WavePlayer_10, WavePlayer_11, WavePlayer_12], properties {weights [1,1,1]}}
 {id=WavePlayer_10, properties {soundWaveAssetPtr=&/husky_husk_stunned_01}}
 {id=WavePlayer_11, properties {soundWaveAssetPtr=&/husky_husk_stunned_02}}
 {id=WavePlayer_12, properties {soundWaveAssetPtr=&/husky_husk_stunned_03}}
 {id=WavePlayer_5, properties {soundWaveAssetPtr=&husky_husk_stunned_01}}
 {id=WavePlayer_6, properties {soundWaveAssetPtr=&husky_husk_stunned_02}}
 {id=WavePlayer_7, properties {soundWaveAssetPtr=&husky_husk_stunned_03}}
 {id=WavePlayer_8, properties {soundWaveAssetPtr=&/husky_husk_stunned_04}}
 {id=WavePlayer_9, properties {soundWaveAssetPtr=&/husky_husk_stunned_05}}
]
properties {
 attenuationSettings=/Game/Sounds/Attenuation/Fort_Monster_Husk_Att
 bOverrideConcurrency=true
 maxDistance=2800
 soundClassObject=/Game/Sounds/FortSoundClassesAndModes/Fort_Enemy_Damaged_PSM
 volumeMultiplier=1
 attenuationOverrides {distanceAlgorithm=NaturalSound, falloffDistance=4000, attenuationShapeExtents {x=1000}}
 concurrencyOverrides {resolutionRule=PreventNew}
}
